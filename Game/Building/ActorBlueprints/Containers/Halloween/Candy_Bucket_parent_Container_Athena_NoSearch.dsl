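; Parent Blueprint for Halloween candy bucket containers, judging by its name a no-search container variant in the Athena mode.
; It overrides the searched-state event, calls the parent implementation, then destroys the actor if the destroy-on-search flag is set.
; The construction script is empty, so all other behavior comes from the parent class.
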(event EventOnSetSearched
  (|CallParentFunction)
  (if (Variables|Container|GetbDestroyContainerOnSearch)
    (Actor|DestroyActor)))

(fn ConstructionScript)
